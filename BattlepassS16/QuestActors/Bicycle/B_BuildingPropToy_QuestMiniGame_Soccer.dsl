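; Soccer target prop for a quest minigame that gets knocked back when a toy soccer ball hits it.
; On trigger overlap, it scores and destroys the ball, then sends a target-hit quest stat event for the ball owner's player controller.
; Knockback plays a sound, disables trigger collision and rotates the target mesh by timeline.
; After a random delay, a reset timeline rotates it back and re-enables query-only collision.
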(event OnComponentBeginOverlap(TriggerBox)
  (params (OverlappedComponent :type Primitive%20Component%20Object%20Reference) (OtherActor :type Actor%20Object%20Reference) (OtherComp :type Primitive%20Component%20Object%20Reference) (OtherBodyIndex :type Integer) (bFromSweep :type Boolean) (SweepResult :type Hit%20Result%20Structure%20%28by%20ref%29))
  ;[1 1 1] On overlap send quest update and knockback target
  (bind _asb_toy_soccer_ball_base (Utilities|Casting|CastToB_Toy_SoccerBall_Base OtherActor))
  (Class|BToySoccerBallBase|PlayScoringCueandDestroy _asb_toy_soccer_ball_base true)
  (Quest|SendComplexCustomStatEvent (Quests|GetQuestManager (Fortnite|GetFortPlayerControllerfromActor (Actor|GetOwner _asb_toy_soccer_ball_base)) "Invalid") 0 "(GameplayTags=((TagName=\"Athena.Quests.SoccerBallToy.TargetHit\")))" "(GameplayTags=)" 1)
  (CallFunction|Knockback))

(event Custom|Knockback
  ;[1 1 1] Disable trigger collision and knockback target
  (Audio|PlaySoundatLocation "/Game/Athena/Sounds/Quests/Athena_ShootingGallery_HitTarget_Cue.Athena_ShootingGallery_HitTarget_Cue" (Transformation|GetActorLocation))
  (Collision|SetCollisionEnabled (Variables|Default|GetTriggerBox) "NoCollision")
  (bind (_direction _rotation) (Timeline|Knockback_Timeline PlayFromStart)
    (:Update
      (Transformation|SetRelativeRotation (Variables|Default|Getshooting_target_01) (Math|Rotator|MakeRotator 0.0 _rotation) false false))
    (:Finished
      (CallFunction|ResetKnockback))))

(event Custom|Reset%20Knockback
  ;[1 1 1] Reset target postion and enable collision
  (Utilities|FlowControl|Delay (Math|Random|RandomFloatinRange 1.0 4.0))
  (bind (_direction _rotation) (Timeline|Reset_Timeline PlayFromStart)
    (:Update
      (Transformation|SetRelativeRotation (Variables|Default|Getshooting_target_01) (Math|Rotator|MakeRotator 0.0 _rotation) false false))
    (:Finished
      (Collision|SetCollisionEnabled (Variables|Default|GetTriggerBox) "QueryOnly"))))

(fn ConstructionScript
  (|CallParentFunction))
